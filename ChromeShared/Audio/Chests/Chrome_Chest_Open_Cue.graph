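# Sound cue for the Chrome chest opening sound, classed as a loot reward sound.
# A single wave player feeds a modulator that slightly randomizes pitch on each play while keeping volume constant.
# It uses shared item-interact attenuation and overrides concurrency to cap simultaneous instances at a few, stopping the lowest priority when exceeded.
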SoundCue /ChromeShared/Audio/Chests/Chrome_Chest_Open_Cue
root=Modulator_2
nodes [
 {id=Modulator_2, children [WavePlayer_2], properties {pitchMax=1.1, pitchMin=0.9, volumeMax=1, volumeMin=1}}
 {id=WavePlayer_2, properties {soundWaveAssetPtr=&/Chrome_Chest_Open_01}}
]
properties {
 attenuationSettings=/Game/Sounds/Attenuation/Fort_Item_Interact_Att
 bOverrideConcurrency=true
 maxDistance=1850
 soundClassObject=&Fort_Loot_Reward_PSM
 concurrencyOverrides {maxCount=3, resolutionRule=StopLowestPriority}
}
